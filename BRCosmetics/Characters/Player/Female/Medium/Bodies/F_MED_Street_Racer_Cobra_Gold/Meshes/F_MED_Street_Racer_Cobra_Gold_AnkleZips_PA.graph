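# Physics asset for the ankle zipper pulls on the female medium Street Racer Cobra Gold body cosmetic.
# Small capsule bodies sit on both lower-calf twist bones and on the left and right dynamic zipper bones.
# Each zipper is constrained to its calf bone with limited swing and tighter twist, letting it dangle within a narrow cone.
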PhysicsAsset /BRCosmetics/Characters/Player/Female/Medium/Bodies/F_MED_Street_Racer_Cobra_Gold/Meshes/F_MED_Street_Racer_Cobra_Gold_AnkleZips_PA
bodies {
 calf_twist_01_l [{shapeType=Capsule, length=0.5, radius=0.5}]
 calf_twist_01_r [{shapeType=Capsule, length=0.5, radius=0.5}]
 dyn_zipper_ankle_l [{shapeType=Capsule, length=2.1467, radius=0.8826, center=<2 0 0>, rotation=<rot 90 0 0>}]
 dyn_zipper_ankle_r [{shapeType=Capsule, length=2.1467, radius=0.8801, center=<2 0 0>, rotation=<rot 90 0 0>}]
}
constraints [
 {bone1Name=dyn_zipper_ankle_r, bone2Name=calf_twist_01_r, swing1LimitDegrees=45, swing2LimitDegrees=45, twistLimitDegrees=15}
 {bone1Name=dyn_zipper_ankle_l, bone2Name=calf_twist_01_l, swing1LimitDegrees=45, swing2LimitDegrees=45, twistLimitDegrees=15}
]
