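# Physics asset for the jacket of the female medium Street Racer Cobra Gold body, setting up physical secondary motion for its dyn_jacket bones.
# Seven chains around the jacket each run from a small master capsule through two thin box bodies, with masters hanging off the spine_04 torso capsule on free angular joints.
# Chain joints keep twist tight, with narrower swing on the three back chains.
# One extra free-angular constraint links the back-right tip to the center tip.
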PhysicsAsset /BRCosmetics/Characters/Player/Female/Medium/Bodies/F_MED_Street_Racer_Cobra_Gold/Meshes/F_MED_Street_Racer_Cobra_Gold_Jacket_PA
bodies {
 dyn_jacket_bl_01 [{shapeType=Box, extentX=7, extentY=10, extentZ=1, center=<3 0 0>}]
 dyn_jacket_bl_02 [{shapeType=Box, extentX=7, extentY=10, extentZ=1, center=<3 0 0>}]
 dyn_jacket_bl_MSTR [{shapeType=Capsule, length=0.505, radius=0.505}]
 dyn_jacket_br_01 [{shapeType=Box, extentX=7, extentY=10, extentZ=1, center=<3 0 0>}]
 dyn_jacket_br_02 [{shapeType=Box, extentX=7, extentY=10, extentZ=1, center=<3 0 0>}]
 dyn_jacket_br_MSTR [{shapeType=Capsule, length=0.505, radius=0.505}]
 dyn_jacket_c_01 [{shapeType=Box, extentX=7, extentY=10, extentZ=1, center=<3 0 0>}]
 dyn_jacket_c_02 [{shapeType=Box, extentX=7, extentY=10, extentZ=1, center=<3 0 0>}]
 dyn_jacket_c_MSTR [{shapeType=Capsule, length=0.5, radius=0.5}]
 dyn_jacket_f_l_01 [{shapeType=Box, extentX=7, extentY=10, extentZ=1, center=<3 0 0>}]
 dyn_jacket_f_l_02 [{shapeType=Box, extentX=7, extentY=10, extentZ=1, center=<3 0 0>}]
 dyn_jacket_f_l_MSTR [{shapeType=Capsule, length=0.505, radius=0.505}]
 dyn_jacket_f_r_01 [{shapeType=Box, extentX=7, extentY=10, extentZ=1, center=<3 0 0>}]
 dyn_jacket_f_r_02 [{shapeType=Box, extentX=7, extentY=10, extentZ=1, center=<3 0 0>}]
 dyn_jacket_f_r_MSTR [{shapeType=Capsule, length=0.505, radius=0.505}]
 dyn_jacket_l_01 [{shapeType=Box, extentX=7, extentY=10, extentZ=1, center=<3 0 0>}]
 dyn_jacket_l_02 [{shapeType=Box, extentX=7, extentY=10, extentZ=1, center=<3 0 0>}]
 dyn_jacket_l_MSTR [{shapeType=Capsule, length=0.505, radius=0.505}]
 dyn_jacket_r_01 [{shapeType=Box, extentX=7, extentY=10, extentZ=1, center=<3 0 0>}]
 dyn_jacket_r_02 [{shapeType=Box, extentX=7, extentY=10, extentZ=1, center=<3 0 0>}]
 dyn_jacket_r_MSTR [{shapeType=Capsule, length=0.505, radius=0.505}]
 spine_03 [{shapeType=Capsule, length=0, radius=0.6258, center=<0.359 -5.2985 0.9232>}]
 spine_04 [{shapeType=Capsule, length=6.9725, radius=10.0933, center=<-1.6785 3.0144 0.0286>, rotation=<rot 90 90 85.3963>}]
}
constraints [
 {bone1Name=spine_04, bone2Name=spine_03, swing1LimitDegrees=45, swing2LimitDegrees=45, twistLimitDegrees=45}
 {
  bone1Name=dyn_jacket_f_l_MSTR
  bone2Name=spine_04
  swing1LimitDegrees=45
  swing1Motion=Free
  swing2LimitDegrees=45
  swing2Motion=Free
  twistLimitDegrees=45
  twistMotion=Free
 }
 {bone1Name=dyn_jacket_f_l_01, bone2Name=dyn_jacket_f_l_MSTR, swing1LimitDegrees=45, swing2LimitDegrees=45, twistLimitDegrees=5}
 {bone1Name=dyn_jacket_f_l_02, bone2Name=dyn_jacket_f_l_01, swing1LimitDegrees=45, swing2LimitDegrees=45, twistLimitDegrees=5}
 {
  bone1Name=dyn_jacket_l_MSTR
  bone2Name=spine_04
  swing1LimitDegrees=45
  swing1Motion=Free
  swing2LimitDegrees=45
  swing2Motion=Free
  twistLimitDegrees=45
  twistMotion=Free
 }
 {bone1Name=dyn_jacket_l_01, bone2Name=dyn_jacket_l_MSTR, swing1LimitDegrees=45, swing2LimitDegrees=45, twistLimitDegrees=5}
 {bone1Name=dyn_jacket_l_02, bone2Name=dyn_jacket_l_01, swing1LimitDegrees=45, swing2LimitDegrees=45, twistLimitDegrees=5}
 {
  bone1Name=dyn_jacket_bl_MSTR
  bone2Name=spine_04
  swing1LimitDegrees=45
  swing1Motion=Free
  swing2LimitDegrees=45
  swing2Motion=Free
  twistLimitDegrees=45
  twistMotion=Free
 }
 {bone1Name=dyn_jacket_bl_01, bone2Name=dyn_jacket_bl_MSTR, swing1LimitDegrees=35, swing2LimitDegrees=35, twistLimitDegrees=5}
 {bone1Name=dyn_jacket_bl_02, bone2Name=dyn_jacket_bl_01, swing1LimitDegrees=35, swing2LimitDegrees=35, twistLimitDegrees=5}
 {
  bone1Name=dyn_jacket_br_MSTR
  bone2Name=spine_04
  swing1LimitDegrees=45
  swing1Motion=Free
  swing2LimitDegrees=45
  swing2Motion=Free
  twistLimitDegrees=45
  twistMotion=Free
 }
 {bone1Name=dyn_jacket_br_01, bone2Name=dyn_jacket_br_MSTR, swing1LimitDegrees=35, swing2LimitDegrees=35, twistLimitDegrees=5}
 {bone1Name=dyn_jacket_br_02, bone2Name=dyn_jacket_br_01, swing1LimitDegrees=35, swing2LimitDegrees=35, twistLimitDegrees=5}
 {
  bone1Name=dyn_jacket_r_MSTR
  bone2Name=spine_04
  swing1LimitDegrees=45
  swing1Motion=Free
  swing2LimitDegrees=45
  swing2Motion=Free
  twistLimitDegrees=45
  twistMotion=Free
 }
 {bone1Name=dyn_jacket_r_01, bone2Name=dyn_jacket_r_MSTR, swing1LimitDegrees=45, swing2LimitDegrees=45, twistLimitDegrees=5}
 {bone1Name=dyn_jacket_r_02, bone2Name=dyn_jacket_r_01, swing1LimitDegrees=45, swing2LimitDegrees=45, twistLimitDegrees=5}
 {
  bone1Name=dyn_jacket_f_r_MSTR
  bone2Name=spine_04
  swing1LimitDegrees=45
  swing1Motion=Free
  swing2LimitDegrees=45
  swing2Motion=Free
  twistLimitDegrees=45
  twistMotion=Free
 }
 {bone1Name=dyn_jacket_f_r_01, bone2Name=dyn_jacket_f_r_MSTR, swing1LimitDegrees=45, swing2LimitDegrees=45, twistLimitDegrees=5}
 {bone1Name=dyn_jacket_f_r_02, bone2Name=dyn_jacket_f_r_01, swing1LimitDegrees=45, swing2LimitDegrees=45, twistLimitDegrees=5}
 {
  bone1Name=dyn_jacket_c_MSTR
  bone2Name=spine_04
  swing1LimitDegrees=45
  swing1Motion=Free
  swing2LimitDegrees=45
  swing2Motion=Free
  twistLimitDegrees=45
  twistMotion=Free
 }
 {bone1Name=dyn_jacket_c_01, bone2Name=dyn_jacket_c_MSTR, swing1LimitDegrees=35, swing2LimitDegrees=35, twistLimitDegrees=5}
 {bone1Name=dyn_jacket_c_02, bone2Name=dyn_jacket_c_01, swing1LimitDegrees=35, swing2LimitDegrees=35, twistLimitDegrees=5}
 {
  bone1Name=dyn_jacket_br_02
  bone2Name=dyn_jacket_c_02
  swing1LimitDegrees=45
  swing1Motion=Free
  swing2LimitDegrees=45
  swing2Motion=Free
  twistLimitDegrees=45
  twistMotion=Free
 }
]
