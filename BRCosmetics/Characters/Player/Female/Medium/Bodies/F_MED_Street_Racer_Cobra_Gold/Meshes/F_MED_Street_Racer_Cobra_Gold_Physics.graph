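# Physics asset for the female medium Street Racer Cobra Gold body cosmetic, covering dynamic fur and zipper pouch bones.
# Small fur capsules chain from spine_05 through master bones to limited-swing tips, with extra driver bones in the back-left and back-right chains.
# Most master and driver joints swing and twist freely, but the back-right master stays at default motion.
# Pouch boxes hang off upper-arm twist capsules on tight joints, with left twist locked.
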PhysicsAsset /BRCosmetics/Characters/Player/Female/Medium/Bodies/F_MED_Street_Racer_Cobra_Gold/Meshes/F_MED_Street_Racer_Cobra_Gold_Physics
bodies {
 dyn_fur_bl_01 [{shapeType=Capsule, length=2, radius=1, center=<1 0 0>, rotation=<rot 90 0 0>}]
 dyn_fur_bl_DRVR [{shapeType=Capsule, length=0.5, radius=0.5}]
 dyn_fur_bl_MSTR [{shapeType=Capsule, length=0.5, radius=0.5}]
 dyn_fur_br_01 [{shapeType=Capsule, length=2, radius=1, center=<1 0 0>, rotation=<rot 90 0 0>}]
 dyn_fur_br_DRVR [{shapeType=Capsule, length=0.5, radius=0.5}]
 dyn_fur_br_MSTR [{shapeType=Capsule, length=0.5, radius=0.5}]
 dyn_fur_c_01 [{shapeType=Capsule, length=2, radius=1, center=<2.0252 -0.528 -0.2076>, rotation=<rot 90 0 0>}]
 dyn_fur_c_MSTR [{shapeType=Capsule, length=0.5, radius=0.5}]
 dyn_fur_c_l_01 [{shapeType=Capsule, length=2, radius=1, center=<1 0.1816 0>, rotation=<rot 90 0 0>}]
 dyn_fur_c_l_MSTR [{shapeType=Capsule, length=0.5, radius=0.5}]
 dyn_fur_c_r_01 [{shapeType=Capsule, length=2, radius=1, center=<0.9761 -0.09 -0.0043>, rotation=<rot 90 0 0>}]
 dyn_fur_c_r_MSTR [{shapeType=Capsule, length=0.5, radius=0.5}]
 dyn_fur_cb_01 [{shapeType=Capsule, length=2, radius=1, center=<2.4636 0.222 -0.0001>, rotation=<rot 90 0 0>}]
 dyn_fur_cb_MSTR [{shapeType=Capsule, length=0.5, radius=0.5}]
 dyn_fur_cb_l_01 [{shapeType=Capsule, length=2, radius=1, center=<1 0 0>, rotation=<rot 90 0 0>}]
 dyn_fur_cb_l_MSTR [{shapeType=Capsule, length=0.5, radius=0.5}]
 dyn_fur_cb_r_01 [{shapeType=Capsule, length=2, radius=1, center=<1 0 0>, rotation=<rot 90 0 0>}]
 dyn_fur_cb_r_MSTR [{shapeType=Capsule, length=0.5, radius=0.5}]
 dyn_fur_ct_01 [{shapeType=Capsule, length=2, radius=1, center=<1.082 0 0>, rotation=<rot 90 0 0>}]
 dyn_fur_ct_MSTR [{shapeType=Capsule, length=0.5, radius=0.5}]
 dyn_fur_f_l_01 [{shapeType=Capsule, length=2, radius=1, center=<1 0 0>, rotation=<rot 90 0 0>}]
 dyn_fur_f_l_MSTR [{shapeType=Capsule, length=0.505, radius=0.505}]
 dyn_fur_f_r_01 [{shapeType=Capsule, length=2, radius=1, center=<1 0 0>, rotation=<rot 90 0 0>}]
 dyn_fur_f_r_MSTR [{shapeType=Capsule, length=0.505, radius=0.505}]
 dyn_fur_l_01 [{shapeType=Capsule, length=2, radius=1, center=<1 0 0>, rotation=<rot 90 0 0>}]
 dyn_fur_l_MSTR [{shapeType=Capsule, length=0.505, radius=0.505}]
 dyn_fur_r_01 [{shapeType=Capsule, length=2, radius=1, center=<1 0 0>, rotation=<rot 81.9307 -179.9999 175.0738>}]
 dyn_fur_r_MSTR [{shapeType=Capsule, length=0.505, radius=0.505}]
 dyn_zipper_pouch_l [{shapeType=Box, extentX=2.7623, extentY=2.252, extentZ=0.6718, center=<1.6204 -0.4246 0.7149>, rotation=<rot 25.8825 -0.0823 -9.014>}]
 dyn_zipper_pouch_r [{shapeType=Box, extentX=2.7261, extentY=2.1861, extentZ=0.5, center=<1.1558 0.1319 0.4369>, rotation=<rot 21.7025 -1.5604 12.518>}]
 spine_05 [{shapeType=Capsule, length=0.3804, radius=2.3656}]
 upperarm_twist_01_l [{shapeType=Capsule, length=4.8619, radius=5.6681, center=<2.3915 0 0>, rotation=<rot 90 0 0>}]
 upperarm_twist_01_r [{shapeType=Capsule, length=4.528, radius=5.7414, center=<-0.6928 0 0>, rotation=<rot 90 0 0>}]
}
constraints [
 {
  bone1Name=dyn_fur_f_l_MSTR
  bone2Name=spine_05
  swing1LimitDegrees=45
  swing1Motion=Free
  swing2LimitDegrees=45
  swing2Motion=Free
  twistLimitDegrees=45
  twistMotion=Free
 }
 {bone1Name=dyn_fur_f_l_01, bone2Name=dyn_fur_f_l_MSTR, swing1LimitDegrees=20, swing2LimitDegrees=20, twistLimitDegrees=5}
 {
  bone1Name=dyn_fur_l_MSTR
  bone2Name=spine_05
  swing1LimitDegrees=45
  swing1Motion=Free
  swing2LimitDegrees=45
  swing2Motion=Free
  twistLimitDegrees=45
  twistMotion=Free
 }
 {bone1Name=dyn_fur_l_01, bone2Name=dyn_fur_l_MSTR, swing1LimitDegrees=20, swing2LimitDegrees=20, twistLimitDegrees=5}
 {
  bone1Name=dyn_fur_bl_MSTR
  bone2Name=spine_05
  swing1LimitDegrees=45
  swing1Motion=Free
  swing2LimitDegrees=45
  swing2Motion=Free
  twistLimitDegrees=45
  twistMotion=Free
 }
 {bone1Name=dyn_fur_bl_01, bone2Name=dyn_fur_bl_DRVR, swing1LimitDegrees=20, swing2LimitDegrees=20, twistLimitDegrees=5}
 {bone1Name=dyn_fur_br_MSTR, bone2Name=spine_05, swing1LimitDegrees=45, swing2LimitDegrees=45, twistLimitDegrees=15}
 {bone1Name=dyn_fur_br_01, bone2Name=dyn_fur_br_DRVR, swing1LimitDegrees=20, swing2LimitDegrees=20, twistLimitDegrees=5}
 {
  bone1Name=dyn_fur_r_MSTR
  bone2Name=spine_05
  swing1LimitDegrees=45
  swing1Motion=Free
  swing2LimitDegrees=45
  swing2Motion=Free
  twistLimitDegrees=45
  twistMotion=Free
 }
 {
  bone1Name=dyn_fur_f_r_MSTR
  bone2Name=spine_05
  swing1LimitDegrees=45
  swing1Motion=Free
  swing2LimitDegrees=45
  swing2Motion=Free
  twistLimitDegrees=45
  twistMotion=Free
 }
 {bone1Name=dyn_fur_f_r_01, bone2Name=dyn_fur_f_r_MSTR, swing1LimitDegrees=20, swing2LimitDegrees=20, twistLimitDegrees=5}
 {
  bone1Name=dyn_fur_c_MSTR
  bone2Name=spine_05
  swing1LimitDegrees=45
  swing1Motion=Free
  swing2LimitDegrees=45
  swing2Motion=Free
  twistLimitDegrees=45
  twistMotion=Free
 }
 {bone1Name=dyn_fur_c_01, bone2Name=dyn_fur_c_MSTR, swing1LimitDegrees=20, swing2LimitDegrees=20, twistLimitDegrees=5}
 {
  bone1Name=dyn_fur_cb_MSTR
  bone2Name=spine_05
  swing1LimitDegrees=45
  swing1Motion=Free
  swing2LimitDegrees=45
  swing2Motion=Free
  twistLimitDegrees=45
  twistMotion=Free
 }
 {bone1Name=dyn_fur_cb_01, bone2Name=dyn_fur_cb_MSTR, swing1LimitDegrees=20, swing2LimitDegrees=20, twistLimitDegrees=5}
 {bone1Name=dyn_zipper_pouch_r, bone2Name=upperarm_twist_01_r, swing1LimitDegrees=5, swing2LimitDegrees=20, twistLimitDegrees=5}
 {
  bone1Name=dyn_zipper_pouch_l
  bone2Name=upperarm_twist_01_l
  swing1LimitDegrees=5
  swing2LimitDegrees=30
  twistLimitDegrees=5
  twistMotion=Locked
 }
 {
  bone1Name=upperarm_twist_01_l
  bone2Name=spine_05
  swing1LimitDegrees=45
  swing1Motion=Free
  swing2LimitDegrees=45
  swing2Motion=Free
  twistLimitDegrees=45
  twistMotion=Free
 }
 {
  bone1Name=upperarm_twist_01_r
  bone2Name=spine_05
  swing1LimitDegrees=45
  swing1Motion=Free
  swing2LimitDegrees=45
  swing2Motion=Free
  twistLimitDegrees=45
  twistMotion=Free
 }
 {bone1Name=dyn_fur_ct_01, bone2Name=dyn_fur_ct_MSTR, swing1LimitDegrees=20, swing2LimitDegrees=20, twistLimitDegrees=5}
 {
  bone1Name=dyn_fur_ct_MSTR
  bone2Name=spine_05
  swing1LimitDegrees=45
  swing1Motion=Free
  swing2LimitDegrees=45
  swing2Motion=Free
  twistLimitDegrees=45
  twistMotion=Free
 }
 {bone1Name=dyn_fur_r_01, bone2Name=dyn_fur_r_MSTR, swing1LimitDegrees=10, swing2LimitDegrees=10, twistLimitDegrees=5}
 {
  bone1Name=dyn_fur_cb_l_MSTR
  bone2Name=spine_05
  swing1LimitDegrees=45
  swing1Motion=Free
  swing2LimitDegrees=45
  swing2Motion=Free
  twistLimitDegrees=45
  twistMotion=Free
 }
 {bone1Name=dyn_fur_cb_l_01, bone2Name=dyn_fur_cb_l_MSTR, swing1LimitDegrees=30, swing2LimitDegrees=30, twistLimitDegrees=5}
 {
  bone1Name=dyn_fur_cb_r_MSTR
  bone2Name=spine_05
  swing1LimitDegrees=45
  swing1Motion=Free
  swing2LimitDegrees=45
  swing2Motion=Free
  twistLimitDegrees=45
  twistMotion=Free
 }
 {bone1Name=dyn_fur_cb_r_01, bone2Name=dyn_fur_cb_r_MSTR, swing1LimitDegrees=30, swing2LimitDegrees=30, twistLimitDegrees=5}
 {
  bone1Name=dyn_fur_c_l_MSTR
  bone2Name=spine_05
  swing1LimitDegrees=45
  swing1Motion=Free
  swing2LimitDegrees=45
  swing2Motion=Free
  twistLimitDegrees=45
  twistMotion=Free
 }
 {bone1Name=dyn_fur_c_l_01, bone2Name=dyn_fur_c_l_MSTR, swing1LimitDegrees=45, swing2LimitDegrees=45, twistLimitDegrees=15}
 {
  bone1Name=dyn_fur_c_r_MSTR
  bone2Name=spine_05
  swing1LimitDegrees=45
  swing1Motion=Free
  swing2LimitDegrees=45
  swing2Motion=Free
  twistLimitDegrees=45
  twistMotion=Free
 }
 {bone1Name=dyn_fur_c_r_01, bone2Name=dyn_fur_c_r_MSTR, swing1LimitDegrees=45, swing2LimitDegrees=45, twistLimitDegrees=15}
 {
  bone1Name=dyn_fur_br_DRVR
  bone2Name=dyn_fur_br_MSTR
  swing1LimitDegrees=45
  swing1Motion=Free
  swing2LimitDegrees=45
  swing2Motion=Free
  twistLimitDegrees=45
  twistMotion=Free
 }
 {
  bone1Name=dyn_fur_bl_DRVR
  bone2Name=dyn_fur_bl_MSTR
  swing1LimitDegrees=45
  swing1Motion=Free
  swing2LimitDegrees=45
  swing2Motion=Free
  twistLimitDegrees=45
  twistMotion=Free
 }
]
