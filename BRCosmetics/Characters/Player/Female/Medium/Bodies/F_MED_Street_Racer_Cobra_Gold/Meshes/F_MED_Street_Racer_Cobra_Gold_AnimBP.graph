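# Post-process style AnimBP for the female Street Racer Cobra Gold body, adding secondary motion to a pose copied from the attached parent mesh.
# A thigh constraint and additive fur-bone offsets, weighted by upper-arm yaw measured each update, feed three chained RigidBody sims using jacket, body and ankle-zip physics assets.
# Delayed jump events toggle upward jacket force, skydiving force drives the body sim, and initialization resets dynamics.
# Three further fur offset nodes are unconnected.
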AnimBlueprint /BRCosmetics/Characters/Player/Female/Medium/Bodies/F_MED_Street_Racer_Cobra_Gold/Meshes/F_MED_Street_Racer_Cobra_Gold_AnimBP
target_skeleton=&/F_MED_Street_Racer_Cobra_Gold_Skeleton
graphs [
 AnimGraph {nodes [
   n1 = "Transform (Modify) Bone - Bone: dyn_fur_ct_MSTR"(Alpha=1, Translation="-0.000,-1,0.000", Rotation="0,0,0", Scale="1,1,1") {
    rotationMode=BMM_Additive
    translationMode=BMM_Additive
    alphaBoolBlend {blendOption=Linear}
    boneToModify {boneName=dyn_fur_ct_MSTR}
    translation {y=-1}
   }
   n2 = "Transform (Modify) Bone - Bone: dyn_fur_br_MSTR"(ComponentPose=@n1, Alpha=1, Translation="0.000,0.000,3", Rotation="0,0,0", Scale="1,1,1") {
    rotationMode=BMM_Additive
    translationMode=BMM_Additive
    alphaBoolBlend {blendOption=Linear}
    boneToModify {boneName=dyn_fur_br_MSTR}
    translation {z=3}
   }
   "Transform (Modify) Bone - Bone: dyn_fur_bl_MSTR"(ComponentPose=@n2, Alpha=1, Translation="0.000,-2.000,0.824", Rotation="14.309,0.000,0.000", Scale="1,1,1") {
    rotationMode=BMM_Additive
    translationMode=BMM_Additive
    alphaBoolBlend {blendOption=Linear}
    boneToModify {boneName=dyn_fur_bl_MSTR}
    rotation {pitch=14.309}
    translation {y=-2, z=0.824}
   }
   n3 = "Constraint - thigh_l_ol (RR)"!(
    ComponentPose=Local To Component(LocalPose=Copy Pose From Mesh() {bUseAttachedParent=true})
    Alpha=1
    ConstraintWeights_0=0.7
    ConstraintWeights_1:float=0.3
   ) -> (Pose) {
    alphaBoolBlend {blendOption=Linear}
    boneToModify {boneName=thigh_l_ol}
    constraintSetup [{transformType=Rotation, targetBone {boneName=pelvis}}, {transformType=Rotation, targetBone {boneName=thigh_l}}]
    constraintWeights [1,1]
   }
   n4 = Get arm_up_l() -> (arm_up_l:double)
   n5 = "Transform (Modify) Bone - Bone: dyn_fur_l_MSTR"(ComponentPose=@n3, Alpha=@n4, Translation="3,0.000,0.000", Rotation="0,0,0", Scale="1,1,1") {
    rotationMode=BMM_Additive
    translationMode=BMM_Additive
    alphaBoolBlend {blendOption=Linear}
    boneToModify {boneName=dyn_fur_l_MSTR}
    translation {x=3}
   }
   n6 = "Transform (Modify) Bone - Bone: dyn_fur_f_l_MSTR"(ComponentPose=@n5, Alpha=@n4, Translation="1,0,0", Rotation="-10,0.000,0.000", Scale="1,1,1") {
    rotationMode=BMM_Additive
    translationMode=BMM_Additive
    alphaBoolBlend {blendOption=Linear}
    boneToModify {boneName=dyn_fur_f_l_MSTR}
    rotation {pitch=-10}
    translation {x=1}
   }
   n7 = Get arm_up_r() -> (arm_up_r:double)
   n8 = "Transform (Modify) Bone - Bone: dyn_fur_r_MSTR"(ComponentPose=@n6, Alpha=@n7, Translation="-3,0.000,0.000", Rotation="0,0,0", Scale="1,1,1") {
    rotationMode=BMM_Additive
    translationMode=BMM_Additive
    alphaBoolBlend {blendOption=Linear}
    boneToModify {boneName=dyn_fur_r_MSTR}
    translation {x=-3}
   }
   n9 = "Transform (Modify) Bone - Bone: dyn_fur_f_r_MSTR"(ComponentPose=@n8, Alpha=@n7, Translation="-1,0,0", Rotation="10,0,0.000", Scale="1,1,1") {
    rotationMode=BMM_Additive
    translationMode=BMM_Additive
    alphaBoolBlend {blendOption=Linear}
    boneToModify {boneName=dyn_fur_f_r_MSTR}
    rotation {pitch=10}
    translation {x=-1}
   }
   n10 = RigidBody(ComponentPose=@n9, ExternalForce=Get JacketExternalForce() -> (JacketExternalForce:Vector)) {
    bClampLinearTranslationLimitToRefPose=true
    bOverrideWorldGravity=true
    overridePhysicsAsset=&/F_MED_Street_Racer_Cobra_Gold_Jacket_PA
    simulationSpace=BaseBoneSpace
    alphaBoolBlend {blendOption=Linear}
    baseBoneRef {boneName=spine_02}
    componentLinearAccScale=<0.6 0.4 0.4>
    componentLinearVelScale=<0.4 0.4 0.4>
    overrideWorldGravity {z=-980}
   }
   n11 = RigidBody(
    ComponentPose=@n10
    ExternalForce=Break Fort Anim Input Skydiving External Force(FortAnimInput_SkydivingExternalForce=Get SkydivingExternalForceInput()) -> (
     bUseSkydivingVectorForce=~
     SkydivingVectorForce=~
     HeadToPelvisDirection=~
     FloatingMultiplier=~
     FloatingAdditive=~
     DivingMultiplier=~
     DivingAdditive=~
     ParachutingMultiplier=~
     ParachutingAdditive=~
    ).SkydivingVectorForce
   ) {
    bClampLinearTranslationLimitToRefPose=true
    bOverrideWorldGravity=true
    overridePhysicsAsset=&/F_MED_Street_Racer_Cobra_Gold_Physics
    simulationSpace=BaseBoneSpace
    alphaBoolBlend {blendOption=Linear}
    baseBoneRef {boneName=spine_03}
    componentLinearAccScale=<0.25 0.175 0.175>
    componentLinearVelScale=<0.5 0.5 0.5>
    overrideWorldGravity {z=-980}
   }
   n12 = RigidBody(ComponentPose=@n11) {
    bClampLinearTranslationLimitToRefPose=true
    bOverrideWorldGravity=true
    overridePhysicsAsset=&/F_MED_Street_Racer_Cobra_Gold_AnkleZips_PA
    simulationSpace=BaseBoneSpace
    alphaBoolBlend {blendOption=Linear}
    baseBoneRef {boneName=root}
    componentAppliedLinearAccClamp=<3000 3000 3000>
    componentLinearAccScale=<0.3 0.2 0.2>
    componentLinearVelScale=<0.3 0.3 0.3>
    overrideWorldGravity {z=-980}
   }
   Output Pose(Result=Component To Local(ComponentPose=@n12)) {layerGroup=None}
  ]}
 EventGraph {nodes [
   Event BlueprintInitializeAnimation()
   > Delay(Duration=0.125)
   > ResetDynamics(InTeleportType=TeleportPhysics)
   > ResetDynamics(InTeleportType=ResetPhysics)
   Event OnBeginJumping()
   > RetriggerableDelay(Duration=0.25)
   > Set JacketExternalForce(JacketExternalForce:Vector="0,0,1250") -> (Output_Get:Vector)
   Event OnEndJumping()
   > RetriggerableDelay(Duration=0.375)
   > Set JacketExternalForce(JacketExternalForce="0, 0, 0")
   Event BlueprintUpdateAnimation()
   n1 = GetOwnerSkeletalMeshComponent()
   n2 = BreakRotator(InRot=BreakTransform(InTransform=GetDeltaTransformFromRefPose(self=@n1, BoneName=upperarm_r, BaseName=spine_05)).Rotation)
   > Set arm_up_r(arm_up_r:double=MapRangeClamped(Value="Absolute (Float)"(A=@n2.Yaw), InRangeA=35, InRangeB=105)) -> (Output_Get:double)
   n3 = GetOwnerSkeletalMeshComponent()
   n4 = BreakRotator(InRot=BreakTransform(InTransform=GetDeltaTransformFromRefPose(self=@n3, BoneName=upperarm_l, BaseName=spine_05)).Rotation)
   > Set arm_up_l(arm_up_l:double=MapRangeClamped(Value="Absolute (Float)"(A=@n4.Yaw), InRangeA=40, InRangeB=105)) -> (Output_Get:double)
  ]}
]
variables [
 arm_up_l {type=number}
 arm_up_r {type=number}
 JacketExternalForce {type=Vector}
]
