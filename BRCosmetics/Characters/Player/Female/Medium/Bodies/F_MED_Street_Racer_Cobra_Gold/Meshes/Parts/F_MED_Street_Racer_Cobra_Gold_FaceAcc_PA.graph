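# Physics asset for the FaceAcc part of the female medium Street Racer Cobra Gold character, rigging three dynamic hair chains (front, middle, back).
# Capsule bodies form each chain from the head, with angular swing limits and tight twist, except a free back-chain master joint between its first two links.
# Multi-capsule head, two-box right clavicle and small upper-spine bodies provide body shapes, with head and clavicle joined to the spine by free angular constraints.
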PhysicsAsset /BRCosmetics/Characters/Player/Female/Medium/Bodies/F_MED_Street_Racer_Cobra_Gold/Meshes/Parts/F_MED_Street_Racer_Cobra_Gold_FaceAcc_PA
bodies {
 clavicle_r [
  {shapeType=Box, extentX=1.01, extentY=27.5924, extentZ=30.6199, center=<0.0425 0 -0.6319>, rotation=<rot -44.1503 -89.9998 0>}
  {shapeType=Box, extentX=1.5084, extentY=16.6146, extentZ=9.2172, center=<-10.0413 3.2262 -7.1444>, rotation=<rot 0 -9.7636 0>}
 ]
 dyn_hair_b_01 [{shapeType=Capsule, length=2.5, radius=1, center=<1 0 0>, rotation=<rot 90 0 0>}]
 dyn_hair_b_02 [{shapeType=Capsule, length=2.5, radius=1, center=<1 0 0>, rotation=<rot 90 0 0>}]
 dyn_hair_b_03 [{shapeType=Capsule, length=2.5, radius=1, center=<1 0 0>, rotation=<rot 90 0 0>}]
 dyn_hair_b_04 [{shapeType=Capsule, length=2.5, radius=1, center=<1 0 0>, rotation=<rot 90 0 0>}]
 dyn_hair_b_05 [{shapeType=Capsule, length=1.5114, radius=1, center=<1 0 0>, rotation=<rot 90 0 0>}]
 dyn_hair_b_MSTR [{shapeType=Capsule, length=0.5, radius=0.5}]
 dyn_hair_f_01 [{shapeType=Capsule, length=2.0433, radius=0.7975, center=<1 0 0.1974>, rotation=<rot 84.0656 -179.9999 -179.9999>}]
 dyn_hair_f_02 [{shapeType=Capsule, length=1.9554, radius=0.9888, center=<0.9858 -0.2051 0.1277>, rotation=<rot 83.6636 -180 -180>}]
 dyn_hair_f_03 [{shapeType=Capsule, length=2.5346, radius=0.9654, center=<1 0 0>, rotation=<rot 90 0 0>}]
 dyn_hair_f_04 [{shapeType=Capsule, length=2.6051, radius=0.8949, center=<1 0 0.1639>, rotation=<rot 83.157 179.9999 179.9999>}]
 dyn_hair_m_01 [{shapeType=Capsule, length=1.8408, radius=1, center=<0.3197 -0.0913 0.0901>, rotation=<rot 74.2335 179.9998 174.0545>}]
 dyn_hair_m_02 [{shapeType=Capsule, length=1.274, radius=1.6702, center=<0.3719 -0.2489 -0.1761>, rotation=<rot 90 0 0>}]
 dyn_hair_m_03 [{shapeType=Capsule, length=2.2806, radius=1.2194, center=<0.01 -0.1127 0>, rotation=<rot 90 0 0>}]
 dyn_hair_m_04 [{shapeType=Capsule, length=1.7983, radius=0.8438, center=<1 0 0>, rotation=<rot 90 0 0>}]
 head [
  {shapeType=Capsule, length=6.1826, radius=6.2257, center=<3.6745 0.8318 1.2026>, rotation=<rot 90 0 0>}
  {shapeType=Capsule, length=2.8915, radius=6.2952, center=<1.9103 3.3174 0>, rotation=<rot 90 -14.0363 -31.9689>}
  {shapeType=Capsule, length=4.0206, radius=3.555, center=<-1.7275 8.575 0>, rotation=<rot 90 0 -16.2881>}
 ]
 spine_05 [{shapeType=Capsule, length=0.505, radius=0.505, rotation=<rot 0 -90 0>}]
}
constraints [
 {bone1Name=dyn_hair_b_01, bone2Name=head, swing1LimitDegrees=10, swing2LimitDegrees=10, twistLimitDegrees=5}
 {bone1Name=dyn_hair_b_02, bone2Name=dyn_hair_b_MSTR, swing1LimitDegrees=25, swing2LimitDegrees=25, twistLimitDegrees=2}
 {bone1Name=dyn_hair_b_03, bone2Name=dyn_hair_b_02, swing1LimitDegrees=25, swing2LimitDegrees=25, twistLimitDegrees=2}
 {bone1Name=dyn_hair_b_04, bone2Name=dyn_hair_b_03, swing1LimitDegrees=25, swing2LimitDegrees=25, twistLimitDegrees=2}
 {bone1Name=dyn_hair_b_05, bone2Name=dyn_hair_b_04, swing1LimitDegrees=30, swing2LimitDegrees=30, twistLimitDegrees=2}
 {bone1Name=dyn_hair_f_01, bone2Name=head, swing1LimitDegrees=35, swing2LimitDegrees=35, twistLimitDegrees=5}
 {bone1Name=dyn_hair_f_02, bone2Name=dyn_hair_f_01, swing1LimitDegrees=35, swing2LimitDegrees=35, twistLimitDegrees=5}
 {bone1Name=dyn_hair_f_03, bone2Name=dyn_hair_f_02, swing1LimitDegrees=35, swing2LimitDegrees=35, twistLimitDegrees=5}
 {bone1Name=dyn_hair_f_04, bone2Name=dyn_hair_f_03, swing1LimitDegrees=35, swing2LimitDegrees=35, twistLimitDegrees=5}
 {bone1Name=dyn_hair_m_01, bone2Name=head, swing1LimitDegrees=45, swing2LimitDegrees=45, twistLimitDegrees=5}
 {bone1Name=dyn_hair_m_02, bone2Name=dyn_hair_m_01, swing1LimitDegrees=45, swing2LimitDegrees=45, twistLimitDegrees=5}
 {bone1Name=dyn_hair_m_03, bone2Name=dyn_hair_m_02, swing1LimitDegrees=35, swing2LimitDegrees=35, twistLimitDegrees=5}
 {bone1Name=dyn_hair_m_04, bone2Name=dyn_hair_m_03, swing1LimitDegrees=35, swing2LimitDegrees=35, twistLimitDegrees=5}
 {
  bone1Name=clavicle_r
  bone2Name=spine_05
  swing1LimitDegrees=45
  swing1Motion=Free
  swing2LimitDegrees=45
  swing2Motion=Free
  twistLimitDegrees=45
  twistMotion=Free
 }
 {
  bone1Name=head
  bone2Name=spine_05
  swing1LimitDegrees=45
  swing1Motion=Free
  swing2LimitDegrees=45
  swing2Motion=Free
  twistLimitDegrees=45
  twistMotion=Free
 }
 {
  bone1Name=dyn_hair_b_MSTR
  bone2Name=dyn_hair_b_01
  swing1LimitDegrees=45
  swing1Motion=Free
  swing2LimitDegrees=45
  swing2Motion=Free
  twistLimitDegrees=45
  twistMotion=Free
 }
]
