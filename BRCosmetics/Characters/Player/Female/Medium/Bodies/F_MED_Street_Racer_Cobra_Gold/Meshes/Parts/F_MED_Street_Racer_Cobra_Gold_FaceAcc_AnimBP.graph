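# Post-copy AnimBlueprint for the Street Racer Cobra Gold female face accessory, adding physics-driven hair motion.
# It copies the attached parent mesh's pose, then additively rotates dyn_hair_b_MSTR with a slow blend while aiming.
# A RigidBody node then simulates in neck_02 base-bone space with its own physics asset, overridden gravity and damped component motion.
# Each update reads the targeting state to toggle that offset and slightly lower physics alpha while aiming.
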AnimBlueprint /BRCosmetics/Characters/Player/Female/Medium/Bodies/F_MED_Street_Racer_Cobra_Gold/Meshes/Parts/F_MED_Street_Racer_Cobra_Gold_FaceAcc_AnimBP
target_skeleton=&/F_MED_Street_Racer_Cobra_Gold_Skeleton
graphs [
 AnimGraph {nodes [
   n1 = "Transform (Modify) Bone - Bone: dyn_hair_b_MSTR"!(
    ComponentPose=Local To Component(LocalPose=Copy Pose From Mesh() {bUseAttachedParent=true})
    bAlphaBoolEnabled:bool=Get isAiming() -> (isAiming:bool)
    Translation="0,0,0"
    Rotation="20,0.000,-10"
    Scale="1,1,1"
   ) -> (Pose) {
    alphaInputType=Bool
    rotationMode=BMM_Additive
    alphaBoolBlend {blendInTime=10, blendOption=Linear, blendOutTime=10}
    boneToModify {boneName=dyn_hair_b_MSTR}
    rotation {pitch=20, roll=-10}
   }
   n2 = RigidBody(ComponentPose=@n1, Alpha=Get rbAlpha() -> (rbAlpha:double)) {
    bOverrideWorldGravity=true
    overridePhysicsAsset=&/F_MED_Street_Racer_Cobra_Gold_FaceAcc_PA
    simulationSpace=BaseBoneSpace
    alphaBoolBlend {blendOption=Linear}
    baseBoneRef {boneName=neck_02}
    componentLinearAccScale=<0.2 0.4 0.2>
    componentLinearVelScale=<0.2 0.4 0.2>
    overrideWorldGravity {z=-980}
   }
   Output Pose(Result=Component To Local(ComponentPose=@n2)) {layerGroup=None}
  ]}
 EventGraph {nodes [
   Event BlueprintUpdateAnimation()
   n1 = Get bIsTargeting()
   > Set isAiming(isAiming:bool=@n1) -> (Output_Get:bool)
   > n2 = Branch(Condition=@n1)
   > Set rbAlpha(rbAlpha:double=0.75) -> (Output_Get:double)
   Set rbAlpha(execute=@n2.else, rbAlpha=1)
  ]}
]
variables [
 isAiming {type=boolean}
 rbAlpha {type=number, default=1}
]
